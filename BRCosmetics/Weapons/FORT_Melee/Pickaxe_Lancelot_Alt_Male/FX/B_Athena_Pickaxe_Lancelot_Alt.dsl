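; Blueprint for the Lancelot Alt pickaxe cosmetic, driving its particle effects during swings and the Reverence emote.
; Left and right swings set a SwingAlpha bool on the idle and animation trail VFX components, clearing it 1.5 seconds after each swing ends.
; A gameplay cue for the Reverence emote's pickaxe check deactivates the idle VFX when active, reactivates it on removal, and prints a debug string each time.
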(event EventBeginPlay)

(event EventSwingRight
  (|CallParentFunction)
  (Effects|Components|ParticleSystem|SetBoolParameter (CallFunction|GetIdleVFXComponent) "SwingAlpha" true)
  (Effects|Components|ParticleSystem|SetBoolParameter (CallFunction|GetAnimTrailVFXComponent) "SwingAlpha" true))

(event EventTick
  (params (DeltaSeconds :type Float%20%28single-precision%29)))

(event EventSwingRightEnd
  (|CallParentFunction)
  (Utilities|FlowControl|Delay 1.5)
  (Effects|Components|ParticleSystem|SetBoolParameter (CallFunction|GetIdleVFXComponent) "SwingAlpha" false)
  (Effects|Components|ParticleSystem|SetBoolParameter (CallFunction|GetAnimTrailVFXComponent) "SwingAlpha" false))

(event EventSwingLeft
  (|CallParentFunction)
  (Effects|Components|ParticleSystem|SetBoolParameter (CallFunction|GetIdleVFXComponent) "SwingAlpha" true)
  (Effects|Components|ParticleSystem|SetBoolParameter (CallFunction|GetAnimTrailVFXComponent) "SwingAlpha" true))

(event EventSwingLeftEnd
  (|CallParentFunction)
  (Utilities|FlowControl|Delay 1.5)
  (Effects|Components|ParticleSystem|SetBoolParameter (CallFunction|GetIdleVFXComponent) "SwingAlpha" false)
  (Effects|Components|ParticleSystem|SetBoolParameter (CallFunction|GetAnimTrailVFXComponent) "SwingAlpha" false))

(event GameplayCue|GameplayCue.Athena.Emote.Reverence.PickaxeCheck
  (params (EventType :type EGameplayCueEvent%20Enum :default "OnActive") (Parameters :type Gameplay%20Cue%20Parameters%20Structure%20%28by%20ref%29))
  (bind _returnvalue (CallFunction|GetIdleVFXComponent))
  (switch Utilities|FlowControl|Switch|SwitchonEGameplayCueEvent EventType
    (:OnActive
      (Components|Activation|Deactivate _returnvalue)
      (Development|PrintString "Lancelot - Deactivated" false false))
    (:Removed
      (Components|Activation|Activate _returnvalue false)
      (Development|PrintString "Lancelot - Activated" false false))))

(fn ConstructionScript
  (|CallParentFunction))
