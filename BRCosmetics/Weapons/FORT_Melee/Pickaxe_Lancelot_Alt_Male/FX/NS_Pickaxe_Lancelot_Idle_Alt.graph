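# Idle flame effect for the alternate Lancelot pickaxe, tinted by User.Variant_Color and attached to User.SK_Mesh.
# Mesh flames track bones or sockets and store a smoothed lagging position per particle, alongside sprite flames, ashes and lights.
# Delayed one-shot flare and ash bursts accompany the persistent fire, which a system timeline scales.
# SwingAlpha stops continuous sprite spawning and changes main fire opacity; time of day dims mesh flames and lights.
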NiagaraSystem /BRCosmetics/Weapons/FORT_Melee/Pickaxe_Lancelot_Alt_Male/FX/NS_Pickaxe_Lancelot_Idle_Alt
dependencies {
 usedDataInterfaces [
  &NiagaraDataInterfaceArrayFloat
  &NiagaraDataInterfaceArrayFloat3
  &NiagaraDataInterfaceVectorCurve
  &NiagaraDataInterfaceColorCurve
  &NiagaraDataInterfaceCurve
 ]
 usedDynamicInputs [/Game/Effects/Niagara/DynamicInputs/Engine/FN_GetSystemAxis]
 usedModules [&SystemState, &Timeline]
}
emitters {
 Fire {
  emitterUpdateScript [
   &EmitterState(
    Inactive Response=::NewEnumerator1
    Life Cycle Mode=::NewEnumerator1
    Loop Behavior=::NewEnumerator1
    Loop Delay=0.8
    Loop Duration Mode=::NewEnumerator1
    Reset Age On Awaken=true
    Scalability Mode=::NewEnumerator1
    UseLoopDelay=true
   )
   &SpawnBurst_Instantaneous(Spawn Count=3, Use Spawn Probability=false)
  ]
  particleSpawnScript [
   &FN_SetSpawnGroup(Number of Spawn Groups=3)
   /Niagara/Modules/Spawn/Initialization/InitializeParticle(
    Color=$User.Variant_Color
    Mesh Scale=&Multiply_Vector(
     A=&MakeVector(
      X=&SelectFloatFromArray(Direct Array Index=$Particles.SpawnGroup, Float Selection Array=DataInterface() {propertyValues {FloatData [0.05,0.05,0.05]}})
      Y=&SelectFloatFromArray(Direct Array Index=$Particles.SpawnGroup, Float Selection Array=DataInterface() {propertyValues {FloatData [0.05,0.05,0.05]}})
      Z=&SelectFloatFromArray(Direct Array Index=$Particles.SpawnGroup, Float Selection Array=DataInterface() {propertyValues {FloatData [1.7,1.4,2]}})
     )
     B=&VectorFromFloat(Value=1)
    )
    Module.Write Color=true
    Module.Write Lifetime=true
    Module.Write Mass=true
    Module.Write Material Random=true
    Module.Write Position=false
    Module.Write Scale=true
    Module.Write SpriteRotation=false
    Module.Write SpriteSize=false
    Position=&SimulationPosition()
    Sprite Size=<5 5>
   )
   &DynamicMaterialParameters(
    Index 0 Param 1=&SelectFloatFromArray(Direct Array Index=$Particles.SpawnGroup, Float Selection Array=DataInterface() {propertyValues {FloatData [0.4,1,0.666]}})
    Index 0 Param 2=1
    Index 0 Param 3=&SelectFloatFromArray(Direct Array Index=$Particles.SpawnGroup, Float Selection Array=DataInterface() {propertyValues {FloatData [26,23,22]}})
    Module.Param0WriteEnabled=true
    Module.Param1WriteEnabled=true
    Module.Param2WriteEnabled=true
    Module.Param3WriteEnabled=true
   )
   &InitialMeshOrientation(Mesh Orientation Mode=::NewEnumerator7, Use Rotation Vector=true) {module=InitialMeshOrientation001}
  ]
  particleUpdateScript [
   &ParticleState(Kill Particles When Lifetime Has Elapsed=false, Let Infinitely Lived Particles Die When Emitter Deactivates=true)
   &SkeletalMeshLocation(
    Bone / Socket Index=$Particles.SpawnGroup
    "Interpolate Mesh Sampling (Update)"=true
    MaskSpawnGroup=false
    Mesh Sampling Type=::NewEnumerator5
    Sampled Position Offset=&SelectVectorFromArray(
     Direct Array Index=$Particles.SpawnGroup
     Vector Selection Array=DataInterface() {propertyValues {FloatData [<0 0 -25>, <-3.5 0 -15>, <-2 0 -10>]}}
    )
    Skeletal Mesh=$User.SK_Mesh
    Socket Sampling Mode=::NewEnumerator5
   ) {module=SkeletalMeshLocation001}
   &InterpolateOverTime(Rate Of Change=8, Target Vector Value:NiagaraPosition=$Particles.SkeletalMeshLocation001.SampledPosition, Variable Type=::NewEnumerator2)
   /SetVariables(Particles.Interpolate Dynamic=&MakeFloat4FromVectorAndFloat("Vector (RGB)"=$Output.InterpolateOverTime.Moving Average, "Float (Alpha)"=1)) {isSetParameters=true, module=SetVariables_#1}
   &SolveForcesAndVelocity(Clamp Velocity=false, Limit Acceleration=false)
   &ScaleColor(
    Scale Alpha=&Multiply_Float(
     A=&FNTimeOfDayScalar(Night Scale=0.5, Morning Scale=0.7, Evening Scale=0.7)
     B=&Multiply_Float(
      A=$System.Timeline.InverseNormalizedTime
      B=&MakeCustomFloatFromBool(Bool=$User.SwingAlpha, True Float=0.05, False Float=0.1)
     )
    )
    ScaleA=true
    ScaleRGB=true
   )
   &ScaleMeshSize(Scale Factor=&Lerp_Vector(
     A=&VectorFromCurve(VectorCurve=DataInterface() {propertyValues {
        XCurve {keys [
          {arriveTangent=4.1785, interpMode=RCIM_Cubic, leaveTangent=4.1785, tangentMode=RCTM_User, time=0, value=0}
          {arriveTangent=0.9995, interpMode=RCIM_Cubic, leaveTangent=0.9995, tangentMode=RCTM_User, time=0.2223, value=0.6825}
          {arriveTangent=-0.0405, interpMode=RCIM_Cubic, leaveTangent=-0.0405, tangentMode=RCTM_User, time=1, value=1}
         ]}
        YCurve {keys [
          {arriveTangent=4.1785, interpMode=RCIM_Cubic, leaveTangent=4.1785, tangentMode=RCTM_User, time=0, value=0}
          {arriveTangent=0.9995, interpMode=RCIM_Cubic, leaveTangent=0.9995, tangentMode=RCTM_User, time=0.2223, value=0.6825}
          {arriveTangent=-0.0405, interpMode=RCIM_Cubic, leaveTangent=-0.0405, tangentMode=RCTM_User, time=1, value=1}
         ]}
        ZCurve {keys [
          {arriveTangent=4.1785, interpMode=RCIM_Cubic, leaveTangent=4.1785, tangentMode=RCTM_User, time=0, value=0}
          {arriveTangent=0.9995, interpMode=RCIM_Cubic, leaveTangent=0.9995, tangentMode=RCTM_User, time=0.2223, value=0.6825}
          {arriveTangent=-0.0405, interpMode=RCIM_Cubic, leaveTangent=-0.0405, tangentMode=RCTM_User, time=1, value=1}
         ]}
       }})
     B=<1 1 0.1>
     Alpha=$System.Timeline.NormalizedTime
    ))
  ]
  rendererClasses [&NiagaraMeshRendererProperties]
  renderers [{rendererClass=&NiagaraMeshRendererProperties, rendererIndex=0}]
 }
 Fire001 {
  emitterUpdateScript [
   &EmitterState(
    Inactive Response=::NewEnumerator1
    Life Cycle Mode=::NewEnumerator1
    Loop Behavior=::NewEnumerator1
    Loop Delay=0.4
    Loop Duration=2
    Reset Age On Awaken=true
    Scalability Mode=::NewEnumerator1
    UseLoopDelay=true
   )
   &SpawnBurst_Instantaneous(Spawn Count=2, Use Spawn Probability=false)
   &SpawnRate(SpawnRate=50, Use Spawn Probability=false) {enabled=false}
  ]
  particleSpawnScript [
   &FN_SetSpawnGroup(Number of Spawn Groups=2)
   &InitializeParticle(
    Color=$User.Variant_Color
    Lifetime=&SelectFloatFromArray(Direct Array Index=$Particles.SpawnGroup, Float Selection Array=DataInterface() {propertyValues {FloatData [1,1.5]}})
    Mesh Scale=&Multiply_Vector(
     A=&MakeVector(
      X=&SelectFloatFromArray(Direct Array Index=$Particles.SpawnGroup, Float Selection Array=DataInterface() {propertyValues {FloatData [0.1,0.1]}})
      Y=&SelectFloatFromArray(Direct Array Index=$Particles.SpawnGroup, Float Selection Array=DataInterface() {propertyValues {FloatData [0.5,0.25]}})
      Z=&SelectFloatFromArray(Direct Array Index=$Particles.SpawnGroup, Float Selection Array=DataInterface() {propertyValues {FloatData [0.05,0.05]}})
     )
     B=&VectorFromFloat(Value=1)
    )
    Module.Write Color=true
    Module.Write Lifetime=true
    Module.Write Mass=false
    Module.Write Material Random=true
    Module.Write Position=false
    Module.Write Scale=true
    Module.Write SpriteRotation=false
    Module.Write SpriteSize=false
    Position=&SimulationPosition()
    Sprite Size=<5 5>
   )
   &DynamicMaterialParameters(
    Index 0 Param 1=&SelectFloatFromArray(Direct Array Index=$Particles.SpawnGroup, Float Selection Array=DataInterface() {propertyValues {FloatData [0.4,0.2,0.4]}})
    Index 0 Param 2=1
    Index 0 Param 3=&SelectFloatFromArray(Direct Array Index=$Particles.SpawnGroup, Float Selection Array=DataInterface() {propertyValues {FloatData [25,20,25]}})
    Module.Param0WriteEnabled=true
    Module.Param1WriteEnabled=true
    Module.Param2WriteEnabled=true
    Module.Param3WriteEnabled=true
   )
   &InitialMeshOrientation(Mesh Axis To Orient=<0 1 0>, Mesh Orientation Mode=::NewEnumerator2, Use Rotation Vector=true) {module=InitialMeshOrientation001}
   &CameraOffset()
  ]
  particleUpdateScript [
   &ParticleState()
   &SkeletalMeshLocation(
    Bone / Socket Index=$Particles.SpawnGroup
    "Interpolate Mesh Sampling (Update)"=true
    MaskSpawnGroup=false
    Mesh Sampling Type=::NewEnumerator5
    Sampled Position Offset=&SelectVectorFromArray(Direct Array Index=$Particles.SpawnGroup, Vector Selection Array=DataInterface() {propertyValues {FloatData [<-1 0 -35>, <0 0 -25>]}})
    Skeletal Mesh=$User.SK_Mesh
    Socket Sampling Mode=::NewEnumerator5
   ) {module=SkeletalMeshLocation001}
   &InterpolateOverTime(Rate Of Change=12, Target Vector Value:NiagaraPosition=$Particles.SkeletalMeshLocation001.SampledPosition, Variable Type=::NewEnumerator2)
   /SetVariables(Particles.Interpolate Dynamic=&MakeFloat4FromVectorAndFloat("Vector (RGB)"=$Output.InterpolateOverTime.Moving Average, "Float (Alpha)"=1)) {isSetParameters=true, module=SetVariables_#1}
   &SolveForcesAndVelocity(Clamp Velocity=false, Limit Acceleration=false)
   &ScaleColor(
    Scale Alpha=&Multiply_Float(
     A=&FNTimeOfDayScalar(Night Scale=0.5, Morning Scale=0.7, Evening Scale=0.7)
     B=&FloatFromCurve(
      FloatCurve=DataInterface() {propertyValues {Curve {keys [
          {arriveTangent=87.2096, interpMode=RCIM_Cubic, leaveTangent=87.2096, tangentMode=RCTM_User, time=0, value=0}
          {arriveTangent=-14.4153, interpMode=RCIM_Cubic, leaveTangent=-14.4153, tangentMode=RCTM_User, time=0.0913, value=5}
          {arriveTangent=-0.7363, interpMode=RCIM_Cubic, leaveTangent=-0.7363, tangentMode=RCTM_User, time=1, value=0}
         ]}}}
      Scale Curve=0.35
     )
    )
    ScaleA=true
    ScaleRGB=true
   )
   &ScaleMeshSize(Scale Factor=&RandomRangeVector(
     Minimum=&VectorFromCurve(
      VectorCurve=DataInterface() {propertyValues {
        XCurve {keys [
          <at -0.001 0.641 RCIM_Cubic>
          {arriveTangent=0.133, interpMode=RCIM_Cubic, leaveTangent=0.133, time=0.1912, value=0.9517}
          {arriveTangent=1.0416, interpMode=RCIM_Cubic, leaveTangent=1.0416, time=0.38, value=0.6917}
          {arriveTangent=0.4973, interpMode=RCIM_Cubic, leaveTangent=0.4973, time=0.6533, value=1.433}
          <at 1 1 RCIM_Cubic>
         ]}
        YCurve {keys [
          <at -0.001 0.641 RCIM_Cubic>
          {arriveTangent=0.109, interpMode=RCIM_Cubic, leaveTangent=0.109, time=0.1902, value=0.958}
          {arriveTangent=0.9567, interpMode=RCIM_Cubic, leaveTangent=0.9567, time=0.3771, value=0.6822}
          {arriveTangent=0.5102, interpMode=RCIM_Cubic, leaveTangent=0.5102, time=0.6571, value=1.4046}
          <at 1 1 RCIM_Cubic>
         ]}
        ZCurve {keys [{arriveTangent=-0.0908, interpMode=RCIM_Cubic, leaveTangent=-0.0908, tangentMode=RCTM_User, time=0, value=0}, <at 1 10 RCIM_Cubic>]}
       }}
      Scale Curve=&MakeVector(
       X=&RandomRangeFloat(Minimum=0.8, Maximum=1)
       Y=&RandomRangeFloat(Minimum=0.8, Maximum=1)
       Z=&RandomRangeFloat(Minimum=1.2, Maximum=1.4)
      )
     )
     Maximum=&VectorFromCurve(
      VectorCurve=DataInterface() {propertyValues {
        XCurve {keys [
          <at 0.0016 0.7213 RCIM_Cubic>
          {arriveTangent=0.1061, interpMode=RCIM_Cubic, leaveTangent=0.1061, time=0.2188, value=0.8087}
          {arriveTangent=0.7814, interpMode=RCIM_Cubic, leaveTangent=0.7814, time=0.4318, value=0.767}
          {arriveTangent=0.6278, interpMode=RCIM_Cubic, leaveTangent=0.6278, time=0.5972, value=1.1044}
          <at 0.9948 1.1204 RCIM_Cubic>
         ]}
        YCurve {keys [
          <at 0.0016 0.7213 RCIM_Cubic>
          {arriveTangent=0.0848, interpMode=RCIM_Cubic, leaveTangent=0.0848, time=0.2179, value=0.815}
          {arriveTangent=0.6812, interpMode=RCIM_Cubic, leaveTangent=0.6812, time=0.4289, value=0.7575}
          {arriveTangent=0.6413, interpMode=RCIM_Cubic, leaveTangent=0.6413, time=0.601, value=1.076}
          <at 0.9948 1.1204 RCIM_Cubic>
         ]}
        ZCurve {keys [{arriveTangent=-0.0908, interpMode=RCIM_Cubic, leaveTangent=-0.0908, tangentMode=RCTM_User, time=0, value=0}, <at 1 10 RCIM_Cubic>]}
       }}
      Scale Curve=&MakeVector(
       X=&RandomRangeFloat(Minimum=0.8, Maximum=1)
       Y=&RandomRangeFloat(Minimum=0.8, Maximum=1)
       Z=&RandomRangeFloat(Minimum=1.2, Maximum=1.4)
      )
     )
    ))
  ]
  rendererClasses [&NiagaraMeshRendererProperties]
  renderers [{rendererClass=&NiagaraMeshRendererProperties, rendererIndex=0}]
 }
 Fire003 {
  emitterUpdateScript [
   &EmitterState(
    Inactive Response=::NewEnumerator1
    Life Cycle Mode=::NewEnumerator1
    Loop Behavior=::NewEnumerator1
    Loop Delay=0.35
    Loop Duration=0.9
    Reset Age On Awaken=true
    Scalability Mode=::NewEnumerator1
    UseLoopDelay=true
   )
   &SpawnBurst_Instantaneous(Spawn Count=2, Use Spawn Probability=false)
   &SpawnRate(SpawnRate=50, Use Spawn Probability=false) {enabled=false}
  ]
  particleSpawnScript [
   &FN_SetSpawnGroup(Number of Spawn Groups=2)
   &InitializeParticle(
    Color=$User.Variant_Color
    Lifetime=&SelectFloatFromArray(Direct Array Index=$Particles.SpawnGroup, Float Selection Array=DataInterface() {propertyValues {FloatData [0.8,0.9,0.8,0.6]}})
    Mesh Scale=&Multiply_Vector(
     A=&MakeVector(
      X=&SelectFloatFromArray(Direct Array Index=$Particles.SpawnGroup, Float Selection Array=DataInterface() {propertyValues {FloatData [0.01,0.01,0.1,0.05]}})
      Y=&SelectFloatFromArray(Direct Array Index=$Particles.SpawnGroup, Float Selection Array=DataInterface() {propertyValues {FloatData [0.5,0.25,0.1,0.2]}})
      Z=&SelectFloatFromArray(Direct Array Index=$Particles.SpawnGroup, Float Selection Array=DataInterface() {propertyValues {FloatData [0.1,0.2,0.05,0.1]}})
     )
     B=&VectorFromFloat(Value=1)
    )
    Module.Write Color=true
    Module.Write Lifetime=true
    Module.Write Mass=false
    Module.Write Material Random=true
    Module.Write Position=false
    Module.Write Scale=true
    Module.Write SpriteRotation=false
    Module.Write SpriteSize=false
    Position=&SimulationPosition()
    Sprite Size=<5 5>
   )
   &DynamicMaterialParameters(
    Index 0 Param 1=&SelectFloatFromArray(Direct Array Index=$Particles.SpawnGroup, Float Selection Array=DataInterface() {propertyValues {FloatData [0.4,0.2,0.156,0.33]}})
    Index 0 Param 2=1
    Index 0 Param 3=&SelectFloatFromArray(Direct Array Index=$Particles.SpawnGroup, Float Selection Array=DataInterface() {propertyValues {FloatData [25,22,22,20]}})
    Module.Param0WriteEnabled=true
    Module.Param1WriteEnabled=true
    Module.Param2WriteEnabled=true
    Module.Param3WriteEnabled=true
   )
   &InitialMeshOrientation(Mesh Axis To Orient=<0 1 0>, Mesh Orientation Mode=::NewEnumerator2, Use Rotation Vector=true) {module=InitialMeshOrientation001}
   &CameraOffset(Camera Offset Amount=2.5)
  ]
  particleUpdateScript [
   &ParticleState()
   &SkeletalMeshLocation(
    Bone / Socket Index=$Particles.SpawnGroup
    "Interpolate Mesh Sampling (Update)"=true
    MaskSpawnGroup=false
    Mesh Sampling Type=::NewEnumerator5
    Sampled Position Offset=&SelectVectorFromArray(
     Direct Array Index=$Particles.SpawnGroup
     Vector Selection Array=DataInterface() {propertyValues {FloatData [<-1 0 -55>, <0 0 -45>, <-2 0 -50>, <2 0 -60>]}}
    )
    Skeletal Mesh=$User.SK_Mesh
    Socket Sampling Mode=::NewEnumerator5
   ) {module=SkeletalMeshLocation001}
   &InterpolateOverTime(Rate Of Change=12, Target Vector Value:NiagaraPosition=$Particles.SkeletalMeshLocation001.SampledPosition, Variable Type=::NewEnumerator2)
   /SetVariables(Particles.Interpolate Dynamic=&MakeFloat4FromVectorAndFloat("Vector (RGB)"=$Output.InterpolateOverTime.Moving Average, "Float (Alpha)"=1)) {isSetParameters=true, module=SetVariables_#1}
   &SolveForcesAndVelocity(Clamp Velocity=false, Limit Acceleration=false)
   &ScaleColor(
    Scale Alpha=&Multiply_Float(
     A=&FNTimeOfDayScalar(Night Scale=0.5, Morning Scale=0.7, Evening Scale=0.7)
     B=&Multiply_Float(
      A=7
      B=&FloatFromCurve(
       FloatCurve=DataInterface() {propertyValues {Curve {keys [
           {arriveTangent=87.2096, interpMode=RCIM_Cubic, leaveTangent=87.2096, tangentMode=RCTM_User, time=0, value=0}
           {arriveTangent=6.7352, interpMode=RCIM_Cubic, leaveTangent=6.7352, tangentMode=RCTM_User, time=0.1, value=12}
           {arriveTangent=-33.1056, interpMode=RCIM_Cubic, leaveTangent=-33.1056, time=0.188, value=8.5186}
           {arriveTangent=-13.9181, interpMode=RCIM_Cubic, leaveTangent=-13.9181, tangentMode=RCTM_User, time=0.3456, value=3.8696}
           {arriveTangent=0.0582, interpMode=RCIM_Cubic, leaveTangent=0.0582, tangentMode=RCTM_User, time=1, value=0}
          ]}}}
       Scale Curve=&SelectFloatFromArray(Direct Array Index=$Particles.SpawnGroup, Float Selection Array=DataInterface() {propertyValues {FloatData [0.005,0.005,0.08,0.01]}})
      )
     )
    )
    ScaleA=true
    ScaleRGB=true
   )
   &ScaleMeshSize(Scale Factor=&RandomRangeVector(
     Minimum=&VectorFromCurve(
      VectorCurve=DataInterface() {propertyValues {
        XCurve {keys [
          <at -0.001 0.641 RCIM_Cubic>
          {arriveTangent=-3.7212, interpMode=RCIM_Cubic, leaveTangent=-3.7212, tangentMode=RCTM_User, time=0.0848, value=1.1368}
          {arriveTangent=0.5211, interpMode=RCIM_Cubic, leaveTangent=0.5211, time=0.38, value=0.6917}
          {arriveTangent=0.4973, interpMode=RCIM_Cubic, leaveTangent=0.4973, time=0.6533, value=1.433}
          <at 1 1 RCIM_Cubic>
         ]}
        YCurve {keys [
          <at -0.001 0.641 RCIM_Cubic>
          {arriveTangent=-3.7427, interpMode=RCIM_Cubic, leaveTangent=-3.7427, tangentMode=RCTM_User, time=0.0838, value=1.1431}
          {arriveTangent=0.4563, interpMode=RCIM_Cubic, leaveTangent=0.4563, time=0.3771, value=0.6822}
          {arriveTangent=0.5102, interpMode=RCIM_Cubic, leaveTangent=0.5102, time=0.6571, value=1.4046}
          <at 1 1 RCIM_Cubic>
         ]}
        ZCurve {keys [{arriveTangent=-0.0908, interpMode=RCIM_Cubic, leaveTangent=-0.0908, tangentMode=RCTM_User, time=0, value=0}, <at 1 10 RCIM_Cubic>]}
       }}
      Scale Curve=&MakeVector(
       X=&RandomRangeFloat(Minimum=0.9, Maximum=1)
       Y=&RandomRangeFloat(Minimum=0.9, Maximum=1)
       Z=&RandomRangeFloat(Minimum=1, Maximum=1.2)
      )
     )
     Maximum=&VectorFromCurve(
      VectorCurve=DataInterface() {propertyValues {
        XCurve {keys [
          <at 0.0016 0.7213 RCIM_Cubic>
          {arriveTangent=0.1061, interpMode=RCIM_Cubic, leaveTangent=0.1061, time=0.2188, value=0.8087}
          {arriveTangent=0.7814, interpMode=RCIM_Cubic, leaveTangent=0.7814, time=0.4318, value=0.767}
          {arriveTangent=0.6278, interpMode=RCIM_Cubic, leaveTangent=0.6278, time=0.5972, value=1.1044}
          <at 0.9948 1.1204 RCIM_Cubic>
         ]}
        YCurve {keys [
          <at 0.0016 0.7213 RCIM_Cubic>
          {arriveTangent=0.0848, interpMode=RCIM_Cubic, leaveTangent=0.0848, time=0.2179, value=0.815}
          {arriveTangent=0.6812, interpMode=RCIM_Cubic, leaveTangent=0.6812, time=0.4289, value=0.7575}
          {arriveTangent=0.6413, interpMode=RCIM_Cubic, leaveTangent=0.6413, time=0.601, value=1.076}
          <at 0.9948 1.1204 RCIM_Cubic>
         ]}
        ZCurve {keys [{arriveTangent=-0.0908, interpMode=RCIM_Cubic, leaveTangent=-0.0908, tangentMode=RCTM_User, time=0, value=0}, <at 1 10 RCIM_Cubic>]}
       }}
      Scale Curve=&MakeVector(
       X=&RandomRangeFloat(Minimum=1.1, Maximum=1.2)
       Y=&RandomRangeFloat(Minimum=1.1, Maximum=1.2)
       Z=&RandomRangeFloat(Minimum=1.2, Maximum=1.4)
      )
     )
    ))
  ]
  rendererClasses [&NiagaraMeshRendererProperties]
  renderers [{rendererClass=&NiagaraMeshRendererProperties, rendererIndex=0}]
 }
 Flame_Ashes {
  emitterUpdateScript [
   &EmitterState(Delay First Loop Only=true, Inactive Response=::NewEnumerator1, Life Cycle Mode=::NewEnumerator1, Loop Delay=0.8, UseLoopDelay=true)
   &SpawnRate(
    SpawnRate=&Multiply_Float(A=6, B=&MakeCustomFloatFromBool(Bool=$User.SwingAlpha, True Float=0, False Float=1))
    Use Spawn Probability=false
   )
  ]
  particleSpawnScript [
   &InitializeParticle(
    Color=$User.Variant_Color
    Lifetime=&UniformRangedFloat(Minimum=0.4, Maximum=0.8)
    Module.Write Color=true
    Module.Write Lifetime=true
    Module.Write Mass=false
    Module.Write Material Random=true
    Module.Write Position=false
    Module.Write Scale=false
    Module.Write SpriteRotation=true
    Module.Write SpriteSize=true
    Position=&SimulationPosition()
    Sprite Rotation=&UniformRangedFloat(Minimum=0, Maximum=360)
    Sprite Size=&RandomRangeVector2D(Minimum=<1 1>, Maximum=<1.5 1.5>)
   )
   /SetVariables(Particles.SpawnOffset:Vector3f=&MakeVector(
     X=&RandomRangeFloat(Minimum=-5, Maximum=5)
     Y=&RandomRangeFloat(Minimum=-5, Maximum=5)
     Z=&RandomRangeFloat(Minimum=-50, Maximum=0)
    )) {isSetParameters=true, module=SetVariables_#2}
   &SkeletalMeshLocation(
    Interpolate Mesh Sampling=false
    MaskSpawnGroup=false
    Mesh Sampling Type=::NewEnumerator5
    Position Sampling=::NewEnumerator1
    Sampled Position Offset=$Particles.SpawnOffset
    Skeletal Mesh=$User.SK_Mesh
    Socket Sampling Mode=::NewEnumerator5
    Source Movement Scale:Vector3f=<0 0 -20>
   )
   &CameraOffset(Camera Offset Amount=&RandomRangeFloat(Minimum=-2.5, Maximum=0))
   &AddVelocity(Velocity=<0 0 200>)
  ]
  particleUpdateScript [
   &ParticleState()
   /Niagara/Modules/Update/Forces/CurlNoiseForce(
    Noise Frequency=0.0125
    Noise Strength=&FloatFromCurve(FloatCurve=DataInterface() {propertyValues {Curve {keys [<at 0 0.7>,<at 1 1>]}}}, Scale Curve=250)
    Pan Noise=true
    Pan Noise Field=<0 0 1>
   )
   &ScaleColor(
    Scale Alpha=&FloatFromCurve(
     FloatCurve=DataInterface() {propertyValues {Curve {keys [
         {arriveTangent=11.9433, interpMode=RCIM_Cubic, leaveTangent=11.9433, tangentMode=RCTM_User, time=0, value=0}
         <at 0.2 1 RCIM_Cubic>
         <at 1 0 RCIM_Cubic>
        ]}}}
     Scale Curve=&Lerp_Float(
      A=1
      B=0
      Alpha=&NormalizeDistanceRange(
       Start Position:NiagaraPosition=/Niagara/DynamicInputs/Helpers/SimulationPosition()
       End Position:NiagaraPosition=$Particles.Position
       Distance:NiagaraFloat=80
       Invert Normalized Range:NiagaraBool=false
      )
     )
    )
    Scale RGB=&MakeVectorFromLinearColorRGB(Color=&ColorFromCurve(ColorCurve=DataInterface() {propertyValues {
        AlphaCurve {keys [<at 0 1 RCIM_Cubic>]}
        BlueCurve {keys [<at 0 4 RCIM_Cubic>,<at 1 0.5 RCIM_Cubic>]}
        GreenCurve {keys [<at 0 4 RCIM_Cubic>,<at 1 0.5 RCIM_Cubic>]}
        RedCurve {keys [<at 0 4 RCIM_Cubic>,<at 1 0.5 RCIM_Cubic>]}
       }}))
    ScaleA=true
    ScaleRGB=true
   )
   &ScaleSpriteSize(Scale Factor=&Vector2DFromFloat(Value=&FloatFromCurve(FloatCurve=DataInterface() {propertyValues {Curve {keys [
          {arriveTangent=7.5764, interpMode=RCIM_Cubic, leaveTangent=7.5765, tangentMode=RCTM_User, time=0, value=0.1}
          {arriveTangent=0.1045, interpMode=RCIM_Cubic, leaveTangent=0.1045, tangentMode=RCTM_User, time=0.25, value=1}
          {arriveTangent=-0.8944, interpMode=RCIM_Cubic, leaveTangent=-0.8944, tangentMode=RCTM_User, time=1, value=0.8}
         ]}}})))
   &DynamicMaterialParameters(
    Index 0 Param 1=&FloatFromCurve(FloatCurve=DataInterface() {propertyValues {Curve {keys [<at 0 0>,<at 1 1>]}}})
    Index 0 Param 2=1
    Index 0 Param 3=1
    Index 0 Param 4=1
    Module.Param0WriteEnabled=true
    Module.Param1WriteEnabled=true
    Module.Param2WriteEnabled=true
    Module.Param3WriteEnabled=true
   )
   &Drag(Drag=4, Use Linear Drag=true, Use Rotational Drag=false)
   &SolveForcesAndVelocity(Clamp Velocity=false, Limit Acceleration=false)
   &SpriteRotationRate(Rotation Rate=250)
  ]
  rendererClasses [&NiagaraSpriteRendererProperties]
  renderers [{rendererClass=&NiagaraSpriteRendererProperties}]
 }
 Flame_Ashes001 {
  emitterUpdateScript [
   &EmitterState(Inactive Response=::NewEnumerator1, Life Cycle Mode=::NewEnumerator1, Loop Behavior=::NewEnumerator1, Loop Delay=0.42, UseLoopDelay=true)
   &SpawnBurst_Instantaneous(Spawn Count=24, Use Spawn Probability=false)
   &SpawnRate(SpawnRate=1000, Use Spawn Probability=false) {enabled=false}
  ]
  particleSpawnScript [
   &InitializeParticle(
    Color=$User.Variant_Color
    Lifetime=&UniformRangedFloat(Minimum=0.3, Maximum=0.6)
    Module.Write Color=true
    Module.Write Lifetime=true
    Module.Write Mass=false
    Module.Write Material Random=true
    Module.Write Position=false
    Module.Write Scale=false
    Module.Write SpriteRotation=true
    Module.Write SpriteSize=true
    Position=/Niagara/Functions/Localspace/SimulationPosition()
    Sprite Size=<1.5 2.5>
   )
   /SetVariables(Particles.SpawnOffset:Vector3f=&MakeVector(
     X=&RandomRangeFloat(Minimum=-5, Maximum=5)
     Y=&RandomRangeFloat(Minimum=-5, Maximum=5)
     Z=&RandomRangeFloat(Minimum=-50, Maximum=0)
    )) {isSetParameters=true, module=SetVariables_#2}
   &SkeletalMeshLocation(
    Interpolate Mesh Sampling=false
    MaskSpawnGroup=false
    Mesh Sampling Type=::NewEnumerator5
    Position Sampling=::NewEnumerator1
    Sampled Position Offset=<0 0 -25>
    Skeletal Mesh=$User.SK_Mesh
    Socket Sampling Mode=::NewEnumerator5
   )
   &CameraOffset(Camera Offset Amount=&RandomRangeFloat(Minimum=-2.5, Maximum=0))
   &AddVelocity(Velocity=<0 0 250>)
   &ShapeLocation(Cylinder Radius=2, MaskSpawnGroup=false, Override Seed=false, Shape Primitive=::NewEnumerator1, Surface Only=false, UseOwnerScale=true)
   &AddVelocity(
    Cone Axis=<0 0 1>
    Distribution Along Cone Axis=0
    Module.UseConeFalloff=false
    Module.UseVelDistribution=false
    Override Seed=false
    Speed Falloff From Cone Axis=0
    Velocity Mode=::NewEnumerator2
    Velocity Speed=200
   ) {module=AddVelocity001}
  ]
  particleUpdateScript [
   &ParticleState()
   &CurlNoiseForce(
    Noise Strength=&FloatFromCurve(FloatCurve=DataInterface() {propertyValues {Curve {keys [<at 0 0.7>,<at 1 1>]}}}, Scale Curve=300)
    Pan Noise=true
    Pan Noise Field=<0 0 1>
   )
   &ScaleColor(
    Scale Alpha=&FloatFromCurve(
     FloatCurve=DataInterface() {propertyValues {Curve {keys [
         {arriveTangent=11.9433, interpMode=RCIM_Cubic, leaveTangent=11.9433, tangentMode=RCTM_User, time=0, value=0}
         <at 0.2 1 RCIM_Cubic>
         <at 1 0 RCIM_Cubic>
        ]}}}
     Scale Curve=&Lerp_Float(
      A=1
      B=0
      Alpha=&NormalizeDistanceRange(
       Start Position:NiagaraPosition=/Niagara/DynamicInputs/Helpers/SimulationPosition()
       End Position:NiagaraPosition=$Particles.Position
       Distance:NiagaraFloat=90
       Invert Normalized Range:NiagaraBool=false
      )
     )
    )
    Scale RGB=&MakeVectorFromLinearColorRGB(Color=&ColorFromCurve(ColorCurve=DataInterface() {propertyValues {
        AlphaCurve {keys [<at 0 1 RCIM_Cubic>]}
        BlueCurve {keys [<at 0 4 RCIM_Cubic>,<at 1 0.5 RCIM_Cubic>]}
        GreenCurve {keys [<at 0 4 RCIM_Cubic>,<at 1 0.5 RCIM_Cubic>]}
        RedCurve {keys [<at 0 4 RCIM_Cubic>,<at 1 0.5 RCIM_Cubic>]}
       }}))
    ScaleA=true
    ScaleRGB=true
   )
   &ScaleSpriteSize(Scale Factor=&Vector2DFromFloat(Value=&FloatFromCurve(FloatCurve=DataInterface() {propertyValues {Curve {keys [
          {arriveTangent=7.5764, interpMode=RCIM_Cubic, leaveTangent=7.5765, tangentMode=RCTM_User, time=0, value=0.1}
          {arriveTangent=0.1045, interpMode=RCIM_Cubic, leaveTangent=0.1045, tangentMode=RCTM_User, time=0.25, value=1}
          {arriveTangent=-0.8944, interpMode=RCIM_Cubic, leaveTangent=-0.8944, tangentMode=RCTM_User, time=1, value=0.8}
         ]}}})))
   &DynamicMaterialParameters(
    Index 0 Param 1=&FloatFromCurve(FloatCurve=DataInterface() {propertyValues {Curve {keys [<at 0 0>,<at 1 1>]}}})
    Index 0 Param 2=1
    Index 0 Param 3=1
    Index 0 Param 4=1
    Module.Param0WriteEnabled=true
    Module.Param1WriteEnabled=true
    Module.Param2WriteEnabled=true
    Module.Param3WriteEnabled=true
   )
   &Drag(Drag=4, Use Linear Drag=true, Use Rotational Drag=false)
   &InheritSourceMovement(Applied Position Scale=<0.8 0.8 0.5>, ApplyForce=false, ApplyPosition=true, ApplyPositionFalloff=false)
   &SolveForcesAndVelocity(Clamp Velocity=false, Limit Acceleration=false)
   &SpriteRotationRate(Rotation Rate=250)
  ]
  rendererClasses [&NiagaraSpriteRendererProperties]
  renderers [{rendererClass=&NiagaraSpriteRendererProperties}]
 }
 Flame_Particles {
  emitterUpdateScript [
   &EmitterState(Delay First Loop Only=true, Inactive Response=::NewEnumerator1, Life Cycle Mode=::NewEnumerator1, Loop Delay=0.8, UseLoopDelay=true)
   &SpawnRate(
    SpawnRate=&Multiply_Float(A=6, B=&MakeCustomFloatFromBool(Bool=$User.SwingAlpha, True Float=0, False Float=1))
    Use Spawn Probability=false
   )
  ]
  particleSpawnScript [
   &InitializeParticle(
    Color=&MultiplyLinearColor(Linear Color=$User.Variant_Color, Scale Factor=<1 1 1 0.2>)
    Lifetime=&Multiply_Float(A=$System.Timeline.NormalizedTime, B=&UniformRangedFloat(Minimum=0.25, Maximum=0.5))
    Module.Write Color=true
    Module.Write Lifetime=true
    Module.Write Mass=false
    Module.Write Material Random=true
    Module.Write Position=false
    Module.Write Scale=false
    Module.Write SpriteRotation=true
    Module.Write SpriteSize=true
    Position=/Niagara/Functions/Localspace/SimulationPosition()
    Sprite Rotation=&UniformRangedFloat(Minimum=0, Maximum=360)
    Sprite Size=&MakeCustomVector2DFromBool(
     Bool=&RandomBool()
     True Vector2D=&Vector2DFromFloat(Value=&RandomRangeFloat(Minimum=16, Maximum=32))
     False Vector2D=&Vector2DFromFloat(Value=&RandomRangeFloat(Minimum=-16, Maximum=-32))
    )
   )
   /SetVariables(Particles.SpawnOffset:Vector3f=&MakeVector(
     X=&RandomRangeFloat(Minimum=-2, Maximum=2)
     Y=&RandomRangeFloat(Minimum=0, Maximum=0)
     Z=&RandomRangeFloat(Minimum=-20, Maximum=10)
    )) {isSetParameters=true, module=SetVariables_#2}
   &SkeletalMeshLocation(
    Interpolate Mesh Sampling=false
    MaskSpawnGroup=false
    Mesh Sampling Type=::NewEnumerator5
    Position Sampling=::NewEnumerator1
    Sampled Position Offset=$Particles.SpawnOffset
    Skeletal Mesh=$User.SK_Mesh
    Socket Sampling Mode=::NewEnumerator5
    Source Movement Scale:Vector3f=<0.05 0.05 0.05>
   )
   &DynamicMaterialParameters(
    Index 0 Param 2=&UniformRangedFloat(Minimum=0, Maximum=1)
    Index 0 Param 3=1
    Index 0 Param 4=&UniformRangedFloat(Minimum=1, Maximum=2)
    Module.Param0WriteEnabled=false
    Module.Param1WriteEnabled=true
    Module.Param2WriteEnabled=false
    Module.Param3WriteEnabled=true
   ) {module=DynamicMaterialParameters002}
   &AccelerationForce(Acceleration=<0 0 50>)
   &InheritVelocity(Module.Limit Speed=true, Velocity Limit=200)
   &ApplyInitialForces(Apply Rotational Force to Rotational Velocity=false)
   &CameraOffset(Camera Offset Amount=&RandomRangeFloat(Minimum=-2.5, Maximum=0))
   &AddVelocity(Velocity=<0 0 100>)
  ]
  particleUpdateScript [
   &ParticleState()
   &ScaleColor(
    Scale Mode=::NewEnumerator1
    Scale RGBA=&Vector4FromCurve(Vector4Curve=DataInterface() {propertyValues {
       AlphaCurve {keys [<at 0.0024 0>,<at 0.4194 1>,<at 0.457 1>,<at 1 0>]}
       BlueCurve {keys [<at 0.3836 1>]}
       GreenCurve {keys [<at 0.3836 1>]}
       RedCurve {keys [<at 0.3836 1>]}
      }})
    ScaleRGBA=true
   )
   /SetVariables(Particles.SubImageIndex=&Add_Float(
     A=&FloatFromCurve(FloatCurve=DataInterface() {propertyValues {Curve {keys [<at 0 0>,<at 1 3>]}}})
     B=&RandomRangeFloat(Minimum=8, Maximum=30)
    )) {isSetParameters=true}
   &ScaleSpriteSize(Scale Factor=&Vector2DFromFloat(Value=&FloatFromCurve(FloatCurve=DataInterface() {propertyValues {Curve {keys [
          {arriveTangent=7.5764, interpMode=RCIM_Cubic, leaveTangent=7.5765, tangentMode=RCTM_User, time=0, value=0.1}
          {arriveTangent=0.1045, interpMode=RCIM_Cubic, leaveTangent=0.1045, tangentMode=RCTM_User, time=0.25, value=1}
          {arriveTangent=-0.8944, interpMode=RCIM_Cubic, leaveTangent=-0.8944, tangentMode=RCTM_User, time=1, value=0.8}
         ]}}})))
   &DynamicMaterialParameters(
    Index 0 Param 1=&FloatFromCurve(FloatCurve=DataInterface() {propertyValues {Curve {keys [{arriveTangent=3.6282, interpMode=RCIM_Cubic, leaveTangent=3.6282, tangentMode=RCTM_User, time=0, value=0.5}, <at 0.5 1>]}}})
    Index 0 Param 2=&RandomRangeFloat(Minimum=0.33, Maximum=1)
    Index 0 Param 3=&FloatFromCurve(FloatCurve=DataInterface() {propertyValues {Curve {keys [
         {arriveTangent=0.1649, interpMode=RCIM_Cubic, leaveTangent=0.1649, tangentMode=RCTM_User, time=0, value=1.25}
         {arriveTangent=2.2011, interpMode=RCIM_Cubic, leaveTangent=2.2011, tangentMode=RCTM_User, time=1, value=2}
        ]}}})
    Index 0 Param 4=1
    Module.Param0WriteEnabled=true
    Module.Param1WriteEnabled=true
    Module.Param2WriteEnabled=true
    Module.Param3WriteEnabled=true
   ) {module=DynamicMaterialParameters001}
   &Drag(Drag=5, Use Linear Drag=true, Use Rotational Drag=false)
   &SolveForcesAndVelocity(Clamp Velocity=false, Limit Acceleration=false)
  ]
  rendererClasses [&NiagaraSpriteRendererProperties]
  renderers [{rendererClass=&NiagaraSpriteRendererProperties}]
 }
 Light {
  emitterUpdateScript [
   &EmitterState(
    Inactive Response=::NewEnumerator1
    Life Cycle Mode=::NewEnumerator1
    Loop Behavior=::NewEnumerator1
    Loop Duration Mode=::NewEnumerator1
    UseLoopDelay=false
   )
   &SpawnBurst_Instantaneous(Spawn Count=2, Use Spawn Probability=false)
  ]
  particleSpawnScript [
   &FN_SetSpawnGroup(Number of Spawn Groups=2)
   /Niagara/Modules/Spawn/Initialization/V2/InitializeParticle(
    Color=&MultiplyLinearColor(Linear Color=$User.Variant_Color, Scale Factor=<2 2 2 1>)
    UsePositionOffset=false
    Write Material Random=true
   )
  ]
  particleUpdateScript [
   &ParticleState(Kill Particles When Lifetime Has Elapsed=false, Let Infinitely Lived Particles Die When Emitter Deactivates=true)
   &SkeletalMeshLocation(
    Apply Sampled Orientation to Mesh Particles=true
    Bone / Socket Index=$Particles.SpawnGroup
    Bone Sampling Mode=::NewEnumerator4
    Inherit Velocity=false
    MaskSpawnGroup=false
    Sampled Position Offset=<0 0 50>
    Sampled Position Offset Coordinate Space:ENiagaraMeshTransforms=::NewEnumerator2
    Skeletal Mesh=$User.SK_Mesh
   )
   &ScaleColor(
    Scale Alpha=&FNTimeOfDayScalar(Night Scale=0.5, Morning Scale=0.7, Evening Scale=0.7)
    Scale RGB=&VectorFromFloat(Value=0.6)
    ScaleA=true
    ScaleRGB=true
   ) {module=ScaleColor001}
  ]
  rendererClasses [&NiagaraLightRendererProperties]
  renderers [{rendererClass=&NiagaraLightRendererProperties, rendererIndex=0}]
 }
}
summary {
 systemName=NS_Pickaxe_Lancelot_Idle_Alt
 emitters [
  {emitterName=Fire, rendererClasses [&NiagaraMeshRendererProperties]}
  {emitterName=Light, rendererClasses [&NiagaraLightRendererProperties]}
  {emitterName=Flame_Particles, rendererClasses [&NiagaraSpriteRendererProperties]}
  {emitterName=Flame_Ashes, rendererClasses [&NiagaraSpriteRendererProperties]}
  {emitterName=Fire001, rendererClasses [&NiagaraMeshRendererProperties]}
  {emitterName=Fire003, rendererClasses [&NiagaraMeshRendererProperties]}
  {emitterName=Flame_Ashes001, rendererClasses [&NiagaraSpriteRendererProperties]}
 ]
 userVariables [
  User.Interpolate_Rate_of_Change {type=NiagaraFloat, defaultValue=12, description=""}
  User.Interpolate_WPO {type=NiagaraFloat, defaultValue=10, description=""}
  User.SK_Mesh {
   type=NiagaraDataInterfaceSkeletalMesh
   defaultValue {
    struct=NiagaraExt_VariableValue_DataInterface
    value {dataInterface=&self:NiagaraDataInterfaceSkeletalMesh_0}
   }
  }
  User.SwingAlpha {type=NiagaraBool}
  User.Variant_Color {type=LinearColor, description="", defaultValue {struct=LinearColor, value=<rgba 0 1 1 1>}}
 ]
}
